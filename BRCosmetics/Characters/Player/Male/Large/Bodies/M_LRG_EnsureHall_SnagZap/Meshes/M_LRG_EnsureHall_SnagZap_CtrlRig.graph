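# Control Rig for the SnagZap large male player body cosmetic, adding procedural corrective motion to its extra bones.
# Setup Leg uses spherical pose readers on leg bones to drive pelvis flaps, knees and right toes, with extra crouch and idle flap offsets.
# It calls Setup Foot, which adjusts foot rotations during skydive.
# Setup Arms then constrains forearm twist bones and offsets arm and shoulder armor bones from pose readers and smoothed Skydive/Emote flags.
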ControlRigBlueprint /BRCosmetics/Characters/Player/Male/Large/Bodies/M_LRG_EnsureHall_SnagZap/Meshes/M_LRG_EnsureHall_SnagZap_CtrlRig
graphs [
 Multiply_5_ContainedGraph {nodes [
   n1 = RigVMFunctionEntryNode()
   RigVMFunctionReturnNode(Result=&RigVMFunction_MathDoubleMul(B=@n1.C, A=&RigVMFunction_MathDoubleMul(A=@n1.A, B=@n1.B).Result).Result)
  ]}
 New Function_ContainedGraph {nodes [
   RigVMFunctionEntryNode()
   > RigVMFunction_Sequence = RigVMAggregateNode()
   n1 = &RigUnit_SphericalPoseReader(
    ActiveRegionScaleFactors=(PositiveWidth=4,NegativeWidth=1,PositiveHeight=4,NegativeHeight=4)
    ActiveRegionSize=0.1
    Debug=(bDrawDebug=false,DebugScale=25,DebugSegments=20,DebugThickness=0.25)
    DriverAxis=(X=-1,Y=0,Z=0)
    DriverItem=(Type=Bone,Name="thigh_l")
    FalloffRegionScaleFactors=(PositiveWidth=1,NegativeWidth=1.3,PositiveHeight=1,NegativeHeight=1)
    FalloffSize=0.48
    FlipHeightScaling=false
    FlipWidthScaling=false
    OptionalParentItem=(Type=Bone,Name="None")
    RotationOffset=(X=90,Y=0,Z=-17)
    ExecutePin=@RigVMFunction_Sequence.A
   )
   > RigVMVariableNode(Value=@n1.OutputParam) {variable=Leg_Front_L}
   > n2 = &RigUnit_SphericalPoseReader(
    ActiveRegionScaleFactors=(PositiveWidth=1,NegativeWidth=1,PositiveHeight=4,NegativeHeight=0.2)
    ActiveRegionSize=0.1
    Debug=(bDrawDebug=false,DebugScale=25,DebugSegments=20,DebugThickness=0.25)
    DriverAxis=(X=-1,Y=0,Z=0)
    DriverItem=(Type=Bone,Name="thigh_l")
    FalloffRegionScaleFactors=(PositiveWidth=1,NegativeWidth=1,PositiveHeight=1,NegativeHeight=1.07)
    FalloffSize=0.46
    FlipHeightScaling=false
    FlipWidthScaling=false
    OptionalParentItem=(Type=Bone,Name="None")
    RotationOffset=(X=180,Y=0,Z=90)
   )
   > RigVMVariableNode(Value=@n2.OutputParam) {variable=Leg_Side_L}
   > n3 = &RigUnit_SphericalPoseReader(
    ActiveRegionScaleFactors=(PositiveWidth=1,NegativeWidth=1,PositiveHeight=1,NegativeHeight=1)
    ActiveRegionSize=0.12
    Debug=(bDrawDebug=false,DebugScale=25,DebugSegments=20,DebugThickness=0.25)
    DriverAxis=(X=1,Y=-90,Z=0)
    DriverItem=(Type=Bone,Name="thigh_l")
    FalloffRegionScaleFactors=(PositiveWidth=1,NegativeWidth=1,PositiveHeight=1,NegativeHeight=1.1)
    FalloffSize=0.3
    FlipHeightScaling=false
    FlipWidthScaling=false
    OptionalParentItem=(Type=Bone,Name="None")
    RotationOffset=(X=-30,Y=0,Z=0)
   )
   > RigVMVariableNode(Value=@n3.OutputParam) {variable=Twist_Out_L}
   > n4 = &RigUnit_SphericalPoseReader(
    ActiveRegionScaleFactors=(PositiveWidth=4,NegativeWidth=1,PositiveHeight=4,NegativeHeight=4)
    ActiveRegionSize=0.1
    Debug=(bDrawDebug=false,DebugScale=25,DebugSegments=20,DebugThickness=0.25)
    DriverAxis=(X=1,Y=0,Z=0)
    DriverItem=(Type=Bone,Name="thigh_r")
    FalloffRegionScaleFactors=(PositiveWidth=1,NegativeWidth=1.3,PositiveHeight=1,NegativeHeight=1)
    FalloffSize=0.48
    FlipHeightScaling=false
    FlipWidthScaling=true
    OptionalParentItem=(Type=Bone,Name="None")
    RotationOffset=(X=-90,Y=0,Z=-17)
   )
   > RigVMVariableNode(Value=@n4.OutputParam) {variable=Leg_Front_R}
   > n5 = &RigUnit_SphericalPoseReader(
    ActiveRegionScaleFactors=(PositiveWidth=1,NegativeWidth=1,PositiveHeight=4,NegativeHeight=0.2)
    ActiveRegionSize=0.1
    Debug=(bDrawDebug=false,DebugScale=25,DebugSegments=20,DebugThickness=0.25)
    DriverAxis=(X=1,Y=0,Z=0)
    DriverItem=(Type=Bone,Name="thigh_r")
    FalloffRegionScaleFactors=(PositiveWidth=1,NegativeWidth=1,PositiveHeight=1,NegativeHeight=1.07)
    FalloffSize=0.46
    FlipHeightScaling=false
    FlipWidthScaling=false
    OptionalParentItem=(Type=Bone,Name="None")
    RotationOffset=(X=0,Y=0,Z=90)
   )
   > RigVMVariableNode(Value=@n5.OutputParam) {variable=Leg_Side_R}
   > n6 = &RigUnit_SphericalPoseReader(
    ActiveRegionScaleFactors=(PositiveWidth=1,NegativeWidth=1,PositiveHeight=1,NegativeHeight=1)
    ActiveRegionSize=0.12
    Debug=(bDrawDebug=false,DebugScale=25,DebugSegments=20,DebugThickness=0.25)
    DriverAxis=(X=1,Y=90,Z=0)
    DriverItem=(Type=Bone,Name="thigh_r")
    FalloffRegionScaleFactors=(PositiveWidth=1,NegativeWidth=1,PositiveHeight=1,NegativeHeight=1.1)
    FalloffSize=0.3
    FlipHeightScaling=false
    FlipWidthScaling=false
    OptionalParentItem=(Type=Bone,Name="None")
    RotationOffset=(X=150,Y=0,Z=0)
   )
   > RigVMVariableNode(Value=@n6.OutputParam) {variable=Twist_Out_R}
   > n7 = &RigUnit_SphericalPoseReader(
    ActiveRegionScaleFactors=(PositiveWidth=1,NegativeWidth=0.3,PositiveHeight=1,NegativeHeight=1)
    ActiveRegionSize=0.1
    Debug=(bDrawDebug=false,bDraw2D=False,bDrawLocalAxes=False,DebugScale=25,DebugSegments=20,DebugThickness=0.25)
    DriverAxis=(X=-1,Y=0,Z=0)
    DriverItem=(Type=Bone,Name="calf_l")
    FalloffRegionScaleFactors=(PositiveWidth=1,NegativeWidth=1,PositiveHeight=1,NegativeHeight=1)
    FalloffSize=0.2
    FlipHeightScaling=False
    FlipWidthScaling=False
    OptionalParentItem=(Type=Bone,Name="None")
    RotationOffset=(X=-90,Y=0,Z=30)
   )
   > &RigUnit_ModifyTransforms(
    ItemToModify=((Item=(Type=Bone,Name="dyn_knee_in_l"),Transform=(Rotation=(X=0,Y=0,Z=0,W=1),Translation=(X=-5.3367,Y=0,Z=0),Scale3D=(X=1,Y=1,Z=1))))
    WeightMinimum=0
    Weight=@n7.OutputParam
   )
   > n8 = &RigUnit_SphericalPoseReader(
    ActiveRegionScaleFactors=(PositiveWidth=1,NegativeWidth=0.3,PositiveHeight=1,NegativeHeight=1)
    ActiveRegionSize=0.1
    Debug=(bDrawDebug=false,bDraw2D=False,bDrawLocalAxes=False,DebugScale=25,DebugSegments=20,DebugThickness=0.25)
    DriverAxis=(X=1,Y=0,Z=0)
    DriverItem=(Type=Bone,Name="calf_r")
    FalloffRegionScaleFactors=(PositiveWidth=1,NegativeWidth=1,PositiveHeight=1,NegativeHeight=1)
    FalloffSize=0.2
    FlipHeightScaling=False
    FlipWidthScaling=true
    OptionalParentItem=(Type=Bone,Name="None")
    RotationOffset=(X=90,Y=0,Z=30)
   )
   > &RigUnit_ModifyTransforms(
    ItemToModify=((Item=(Type=Bone,Name="dyn_knee_in_r"),Transform=(Rotation=(X=0,Y=0,Z=0,W=1),Translation=(X=5.3367,Y=0,Z=0),Scale3D=(X=1,Y=1,Z=1))))
    WeightMinimum=0
    Weight=@n8.OutputParam
   )
   > n9 = &RigUnit_SphericalPoseReader(
    ActiveRegionScaleFactors=(PositiveWidth=0.5,NegativeWidth=1,PositiveHeight=1,NegativeHeight=1)
    ActiveRegionSize=0.1
    Debug=(bDrawDebug=false,bDraw2D=False,bDrawLocalAxes=False,DebugScale=25,DebugSegments=20,DebugThickness=0.25)
    DriverAxis=(X=-1,Y=0,Z=0)
    DriverItem=(Type=Bone,Name="ball_r")
    FalloffRegionScaleFactors=(PositiveWidth=1,NegativeWidth=1,PositiveHeight=1,NegativeHeight=1)
    FalloffSize=0.3
    FlipHeightScaling=False
    FlipWidthScaling=False
    OptionalParentItem=(Type=Bone,Name="None")
    RotationOffset=(X=90,Y=0,Z=40)
   )
   > &RigUnit_ModifyTransforms(
    ItemToModify="((Item=(Type=Bone,Name=\"dyn_toe_1_r\"),Transform=(Rotation=(X=0.0872,Y=0,Z=0,W=0.9962),Translation=(X=0,Y=0,Z=0),Scale3D=(X=1,Y=1,Z=1))),(Item=(Type=Bone,Name=\"dyn_toe_2_r\"),Transform=(Rotation=(X=-0.1736,Y=0,Z=0,W=0.9848),Translation=(X=0,Y=0,Z=0),Scale3D= ...(truncated)"
    WeightMinimum=0
    Weight=@n9.OutputParam
   )
   RigVMFunction_MathBoolToFloat = &RigVMFunction_MathBoolToFloat(Value=RigVMVariableNode() {variable=CrouchingBool}.Value)
   AlphaInterp = &RigVMFunction_AlphaInterp(InterpSpeedDecreasing=10, InterpSpeedIncreasing=10, Scale=1, bClampResult=true, Value=@RigVMFunction_MathBoolToFloat.Result)
   > RigVMVariableNode(Value=@AlphaInterp.Result) {variable=Crouching}
   > n10 = &RigUnit_SphericalPoseReader(
    ActiveRegionScaleFactors=(PositiveWidth=1,NegativeWidth=1,PositiveHeight=1,NegativeHeight=1)
    ActiveRegionSize=0.15
    Debug=(bDrawDebug=false,DebugScale=25,DebugSegments=20,DebugThickness=0.25)
    DriverAxis=(X=1,Y=90,Z=0)
    DriverItem=(Type=Bone,Name="thigh_l")
    FalloffRegionScaleFactors=(PositiveWidth=1,NegativeWidth=1,PositiveHeight=1,NegativeHeight=1.1)
    FalloffSize=0.2
    FlipHeightScaling=false
    FlipWidthScaling=false
    OptionalParentItem=(Type=Bone,Name="None")
    RotationOffset=(X=0,Y=-60,Z=0)
   )
   n11 = &RigVMFunction_AlphaInterp(InterpSpeedDecreasing=5, InterpSpeedIncreasing=20, Scale=1, bClampResult=true, Value=@n10.OutputParam)
   Multiply_4 = RigVMAggregateNode(C=@n11.Result, A=RigVMVariableNode() {variable=Idle}.Value, B=RigVMVariableNode() {variable=Crouching}.Value)
   > &RigUnit_ModifyTransforms(
    ItemToModify=((Item=(Type=Bone,Name="PelvisFlaps_l"),Transform=(Rotation=(X=0,Y=-0.0872,Z=0,W=0.9962),Translation=(X=0,Y=0,Z=0),Scale3D=(X=1,Y=1,Z=1))))
    WeightMinimum=0
    Weight=@Multiply_4.Result
   )
   Setup Foot = RigVMFunctionReferenceNode(ExecuteContext=@RigVMFunction_Sequence.F)
   RigVMFunctionReturnNode(ExecuteContext=@RigVMFunction_Sequence.G)
   &RigUnit_ModifyTransforms(
    ItemToModify=((Item=(Type=Bone,Name="PelvisFlaps_l"),Transform=(Rotation=(X=-0.0121,Y=0.1194,Z=-0.0998,W=0.9877),Translation=(X=2.6604,Y=-2.2654,Z=-0.8141),Scale3D=(X=1,Y=1,Z=1))))
    WeightMinimum=0
    Weight=&RigVMFunction_MathDoubleMul(B=RigVMVariableNode() {variable=Crouching}.Value, A=RigVMVariableNode() {variable=Leg_Front_L}.Value).Result
    ExecutePin=@RigVMFunction_Sequence.C
   )
   Multiply = &RigVMFunction_MathDoubleMul(A=RigVMVariableNode() {variable=Leg_Side_L}.Value, B=RigVMVariableNode() {variable=Crouching}.Value)
   > &RigUnit_ModifyTransforms(
    ItemToModify=((Item=(Type=Bone,Name="PelvisFlaps_l"),Transform=(Rotation=(X=0.0066,Y=-0.1383,Z=0.2502,W=-0.9582),Translation=(X=0,Y=0,Z=0),Scale3D=(X=1,Y=1,Z=1))))
    WeightMinimum=0
    Weight=@Multiply.Result
   )
   > &RigUnit_ModifyTransforms(
    ItemToModify=((Item=(Type=Bone,Name="PelvisFlaps_r"),Transform=(Rotation=(X=0,Y=0,Z=0,W=1),Translation=(X=1.5147,Y=0,Z=-0.0929),Scale3D=(X=1,Y=1,Z=1))))
    WeightMinimum=0
    Weight=&RigVMFunction_MathDoubleMul(B=RigVMVariableNode() {variable=Crouching}.Value, A=RigVMVariableNode() {variable=Leg_Side_R}.Value).Result
   )
   n12 = &RigVMFunction_AlphaInterp(InterpSpeedDecreasing=10, InterpSpeedIncreasing=10, Scale=1, bClampResult=true, Value=RigVMVariableNode() {variable=Idle}.Value)
   > &RigUnit_ModifyTransforms(
    ItemToModify="((Item=(Type=Bone,Name=\"PelvisFlaps_l\"),Transform=(Rotation=(X=-0.0001,Y=-0.0047,Z=-0.024,W=-0.9997),Translation=(X=-1.0366,Y=0.0089,Z=0.0662),Scale3D=(X=1,Y=1,Z=1))),(Item=(Type=Bone,Name=\"PelvisFlaps_r\"),Transform=(Rotation=(X=0,Y=0,Z=0.0376,W=0.9993),Tr ...(truncated)"
    WeightMinimum=0
    Weight=&RigVMFunction_MathDoubleMul(A=@n12.Result, B=RigVMVariableNode() {variable=Crouching}.Value).Result
   )
   &RigUnit_ModifyTransforms(
    ItemToModify=((Item=(Type=Bone,Name="PelvisFlaps_l"),Transform=(Rotation=(X=0.0229,Y=0.0663,Z=-0.1291,W=0.9891),Translation=(X=1.7907,Y=-0.132,Z=0.1954),Scale3D=(X=1,Y=1,Z=1))))
    WeightMinimum=0
    Weight=RigVMVariableNode() {variable=Leg_Front_L}.Value
    ExecutePin=@RigVMFunction_Sequence.B
   )
   > &RigUnit_ModifyTransforms(
    ItemToModify=((Item=(Type=Bone,Name="PelvisFlaps_l"),Transform=(Rotation=(X=0,Y=0.6144,Z=0,W=-0.789),Translation=(X=0,Y=0,Z=0),Scale3D=(X=1,Y=1,Z=1))))
    WeightMinimum=0
    Weight=RigVMVariableNode() {variable=Leg_Side_L}.Value
   )
   > &RigUnit_ModifyTransforms(
    ItemToModify=((Item=(Type=Bone,Name="PelvisFlaps_l"),Transform=(Rotation=(X=0.005,Y=0.0467,Z=0.1061,W=-0.9932),Translation=(X=0,Y=0,Z=0),Scale3D=(X=1,Y=1,Z=1))))
    WeightMinimum=0
    Weight=RigVMVariableNode() {variable=Twist_Out_L}.Value
   )
   > &RigUnit_ModifyTransforms(
    ItemToModify=((Item=(Type=Bone,Name="PelvisFlaps_r"),Transform=(Rotation=(X=0.0146,Y=-0.0394,Z=-0.108,W=0.9933),Translation=(X=-1.2054,Y=0.4084,Z=-0.0902),Scale3D=(X=1,Y=1,Z=1))))
    WeightMinimum=0
    Weight=RigVMVariableNode() {variable=Leg_Front_R}.Value
   )
   > &RigUnit_ModifyTransforms(
    ItemToModify=((Item=(Type=Bone,Name="PelvisFlaps_r"),Transform=(Rotation=(X=0,Y=-0.5442,Z=0,W=0.8389),Translation=(X=0,Y=0,Z=0),Scale3D=(X=1,Y=1,Z=1))))
    WeightMinimum=0
    Weight=RigVMVariableNode() {variable=Leg_Side_R}.Value
   )
   > &RigUnit_ModifyTransforms(
    ItemToModify=((Item=(Type=Bone,Name="PelvisFlaps_r"),Transform=(Rotation=(X=0,Y=0.1056,Z=0,W=-0.9944),Translation=(X=0,Y=0,Z=0),Scale3D=(X=1,Y=1,Z=1))))
    WeightMinimum=0
    Weight=RigVMVariableNode() {variable=Twist_Out_R}.Value
   )
  ]}
 New Function_ContainedGraph {nodes [
   RigVMFunctionEntryNode()
   > RigVMFunction_Sequence = RigVMAggregateNode()
   &RigUnit_RotationConstraintLocalSpaceOffset(
    AdvancedSettings=(InterpolationType=Average,RotationOrderForFilter=XZY)
    Child=(Type=Bone,Name="dyn_extra_lowerarm_l")
    Filter=(bX=true,bY=true,bZ=true)
    Parents=((Item=(Type=Bone,Name="lowerarm_l"),Weight=1))
    Weight=1
    ExecutePin=@RigVMFunction_Sequence.A
   )
   > &RigUnit_RotationConstraintLocalSpaceOffset(
    AdvancedSettings=(InterpolationType=Average,RotationOrderForFilter=XZY)
    Child=(Type=Bone,Name="dyn_extra_lowerarm_r")
    Filter=(bX=true,bY=true,bZ=true)
    Parents=((Item=(Type=Bone,Name="lowerarm_r"),Weight=1))
    Weight=1
   )
   RigVMFunctionReturnNode(ExecuteContext=@RigVMFunction_Sequence.E)
   /Script/ControlRig.RigUnit_PositionConstraintLocalSpaceOffset(
    Child=(Type=Bone,Name="dyn_extra_lowerarm_twist_l")
    Filter=(bX=True,bY=True,bZ=True)
    Parents=((Item=(Type=Bone,Name="hand_l"),Weight=1))
    Weight=1
    bMaintainOffset=True
    ExecutePin=@RigVMFunction_Sequence.D
   ) {is=struct RigVMUnitNode}
   > &RigUnit_PositionConstraintLocalSpaceOffset(
    Child=(Type=Bone,Name="dyn_extra_lowerarm_twist_r")
    Filter=(bX=True,bY=True,bZ=True)
    Parents=((Item=(Type=Bone,Name="hand_r"),Weight=1))
    Weight=1
    bMaintainOffset=True
   )
   > n1 = &RigUnit_SphericalPoseReader(
    ActiveRegionScaleFactors=(PositiveWidth=0.5,NegativeWidth=0.5,PositiveHeight=0.5,NegativeHeight=0.5)
    ActiveRegionSize=0.1
    Debug=(bDrawDebug=false,bDraw2D=False,bDrawLocalAxes=False,DebugScale=25,DebugSegments=20,DebugThickness=0.25)
    DriverAxis=(X=0,Y=0,Z=-1)
    DriverItem=(Type=Bone,Name="hand_l")
    FalloffRegionScaleFactors=(PositiveWidth=1,NegativeWidth=1,PositiveHeight=1,NegativeHeight=1)
    FalloffSize=0.25
    FlipHeightScaling=False
    FlipWidthScaling=False
    OptionalParentItem=(Type=Bone,Name="None")
    RotationOffset=(X=90,Y=0,Z=0)
   )
   > &RigUnit_AimConstraintLocalSpaceOffset(
    Child=(Type=Bone,Name="dyn_extra_lowerarm_twist_l")
    Filter=(bX=false,bY=True,bZ=True)
    Parents=((Item=(Type=Bone,Name="dyn_extra_lowerarm_l"),Weight=1))
    Weight=1
    WorldUp=(Target=(X=0,Y=0,Z=1),Kind=Direction,Space=(Type=None,Name="None"))
   )
   > &RigUnit_ModifyTransforms(
    ItemToModify=((Item=(Type=Bone,Name="dyn_extra_lowerarm_twist_l"),Transform=(Rotation=(X=0,Y=0,Z=0,W=1),Translation=(X=0,Y=0,Z=0),Scale3D=(X=1,Y=1,Z=1))))
    WeightMinimum=0
    Weight=@n1.OutputParam
   )
   > n2 = &RigUnit_SphericalPoseReader(
    ActiveRegionScaleFactors=(PositiveWidth=0.5,NegativeWidth=0.5,PositiveHeight=0.5,NegativeHeight=0.5)
    ActiveRegionSize=0.1
    Debug=(bDrawDebug=false,bDraw2D=False,bDrawLocalAxes=False,DebugScale=25,DebugSegments=20,DebugThickness=0.25)
    DriverAxis=(X=0,Y=0,Z=1)
    DriverItem=(Type=Bone,Name="hand_r")
    FalloffRegionScaleFactors=(PositiveWidth=1,NegativeWidth=1,PositiveHeight=1,NegativeHeight=1)
    FalloffSize=0.25
    FlipHeightScaling=False
    FlipWidthScaling=False
    OptionalParentItem=(Type=Bone,Name="None")
    RotationOffset=(X=-90,Y=0,Z=0)
   )
   > &RigUnit_AimConstraintLocalSpaceOffset(
    Child=(Type=Bone,Name="dyn_extra_lowerarm_twist_r")
    Filter=(bX=false,bY=True,bZ=True)
    Parents=((Item=(Type=Bone,Name="dyn_extra_lowerarm_r"),Weight=1))
    Weight=1
    WorldUp=(Target=(X=0,Y=0,Z=1),Kind=Direction,Space=(Type=None,Name="None"))
   )
   > &RigUnit_ModifyTransforms(
    ItemToModify=((Item=(Type=Bone,Name="dyn_extra_lowerarm_twist_r"),Transform=(Rotation=(X=0,Y=0,Z=0,W=1),Translation=(X=0,Y=0,Z=0),Scale3D=(X=1,Y=1,Z=1))))
    WeightMinimum=0
    Weight=@n2.OutputParam
   )
   n3 = &RigVMFunction_AlphaInterp(
    InterpSpeedDecreasing=10
    InterpSpeedIncreasing=10
    Scale=1
    bClampResult=true
    Value=&RigVMFunction_MathBoolToFloat(Value=RigVMVariableNode() {variable=Skydive}.Value).Result
   )
   &RigUnit_ModifyTransforms(
    ItemToModify="((Item=(Type=Bone,Name=\"dyn_extra_upperarm_l\"),Transform=(Rotation=(X=-0.0868,Y=-0.0076,Z=0.0868,W=0.9924),Translation=(X=3,Y=0,Z=0),Scale3D=(X=1,Y=1,Z=1))),(Item=(Type=Bone,Name=\"dyn_shoulder_armor_l\"),Transform=(Rotation=(X=-0.0421,Y=0.2586,Z=0.0113,W=0. ...(truncated)"
    WeightMinimum=0
    Weight=@n3.Result
    ExecutePin=@RigVMFunction_Sequence.C
   )
   > &RigUnit_ModifyTransforms(
    ItemToModify="((Item=(Type=Bone,Name=\"dyn_extra_upperarm_r\"),Transform=(Rotation=(X=0,Y=0,Z=0.0436,W=0.999),Translation=(X=-4,Y=0,Z=0),Scale3D=(X=1,Y=1,Z=1))),(Item=(Type=Bone,Name=\"dyn_shoulder_armor_r\"),Transform=(Rotation=(X=-0.0842,Y=0.2578,Z=0.0226,W=0.9623),Transl ...(truncated)"
    WeightMinimum=0
    Weight=@n3.Result
   )
   RigVMFunction_MathBoolToFloat = &RigVMFunction_MathBoolToFloat(Value=RigVMVariableNode() {variable=Emote}.Value)
   AlphaInterp = &RigVMFunction_AlphaInterp(InterpSpeedDecreasing=10, InterpSpeedIncreasing=10, Scale=1, bClampResult=true, Value=@RigVMFunction_MathBoolToFloat.Result)
   > &RigUnit_ModifyTransforms(
    ItemToModify="((Item=(Type=Bone,Name=\"dyn_shoulder_armor_l\"),Transform=(Rotation=(X=0.043,Y=0.1735,Z=-0.0076,W=0.9839),Translation=(X=-0.5,Y=0,Z=0),Scale3D=(X=1,Y=1,Z=1))),(Item=(Type=Bone,Name=\"dyn_shoulder_armor_r\"),Transform=(Rotation=(X=-0.0353,Y=0.1752,Z=-0.0353,W= ...(truncated)"
    WeightMinimum=0
    Weight=@AlphaInterp.Result
   )
   n4 = &RigUnit_SphericalPoseReader(
    ActiveRegionScaleFactors=(PositiveWidth=1,NegativeWidth=1,PositiveHeight=1,NegativeHeight=1)
    ActiveRegionSize=0.1
    Debug=(bDrawDebug=false,bDraw2D=False,bDrawLocalAxes=False,DebugScale=25,DebugSegments=20,DebugThickness=0.25)
    DriverAxis=(X=1,Y=0,Z=0)
    DriverItem=(Type=Bone,Name="clavicle_l")
    FalloffRegionScaleFactors=(PositiveWidth=1,NegativeWidth=1,PositiveHeight=1,NegativeHeight=1)
    FalloffSize=0.35
    FlipHeightScaling=False
    FlipWidthScaling=false
    OptionalParentItem=(Type=Bone,Name="None")
    RotationOffset=(X=-25,Y=0,Z=90)
    ExecutePin=@RigVMFunction_Sequence.B
   )
   > &RigUnit_ModifyTransforms(
    ItemToModify=((Item=(Type=Bone,Name="dyn_shoulder_armor_l"),Transform=(Rotation=(X=0,Y=0.1736,Z=0,W=0.9848),Translation=(X=0,Y=0,Z=0),Scale3D=(X=1,Y=1,Z=1))))
    WeightMinimum=0
    Weight=@n4.OutputParam
   )
   > n5 = &RigUnit_SphericalPoseReader(
    ActiveRegionScaleFactors=(PositiveWidth=0.5,NegativeWidth=1.5,PositiveHeight=0.5,NegativeHeight=0.5)
    ActiveRegionSize=0.1
    Debug=(bDrawDebug=false,bDraw2D=False,bDrawLocalAxes=False,DebugScale=25,DebugSegments=20,DebugThickness=0.25)
    DriverAxis=(X=1,Y=0,Z=0)
    DriverItem=(Type=Bone,Name="lowerarm_l")
    FalloffRegionScaleFactors=(PositiveWidth=1,NegativeWidth=1,PositiveHeight=1,NegativeHeight=1)
    FalloffSize=0.5
    FlipHeightScaling=False
    FlipWidthScaling=False
    OptionalParentItem=(Type=Bone,Name="None")
    RotationOffset=(X=90,Y=0,Z=15)
   )
   > &RigUnit_ModifyTransforms(
    ItemToModify="((Item=(Type=Bone,Name=\"dyn_extra_upperarm_twist_l\"),Transform=(Rotation=(X=-0.1736,Y=0,Z=0,W=0.9848),Translation=(X=0,Y=0,Z=0),Scale3D=(X=1,Y=1,Z=1))),(Item=(Type=Bone,Name=\"dyn_extra_lowerarm_start_l\"),Transform=(Rotation=(X=0.0308,Y=-0.0534,Z=-0.2565,W= ...(truncated)"
    WeightMinimum=0
    Weight=@n5.OutputParam
   )
   > n6 = &RigUnit_SphericalPoseReader(
    ActiveRegionScaleFactors=(PositiveWidth=1,NegativeWidth=1,PositiveHeight=1,NegativeHeight=1)
    ActiveRegionSize=0.1
    Debug=(bDrawDebug=false,bDraw2D=False,bDrawLocalAxes=False,DebugScale=25,DebugSegments=20,DebugThickness=0.25)
    DriverAxis=(X=-1,Y=0,Z=0)
    DriverItem=(Type=Bone,Name="clavicle_r")
    FalloffRegionScaleFactors=(PositiveWidth=1,NegativeWidth=1,PositiveHeight=1,NegativeHeight=1)
    FalloffSize=0.35
    FlipHeightScaling=False
    FlipWidthScaling=False
    OptionalParentItem=(Type=Bone,Name="None")
    RotationOffset=(X=180,Y=-25,Z=0)
   )
   > &RigUnit_ModifyTransforms(
    ItemToModify=((Item=(Type=Bone,Name="dyn_shoulder_armor_r"),Transform=(Rotation=(X=0.171,Y=0.171,Z=-0.0302,W=0.9698),Translation=(X=0,Y=0,Z=0),Scale3D=(X=1,Y=1,Z=1))))
    WeightMinimum=0
    Weight=@n6.OutputParam
   )
   > n7 = &RigUnit_SphericalPoseReader(
    ActiveRegionScaleFactors=(PositiveWidth=0.5,NegativeWidth=1.5,PositiveHeight=0.5,NegativeHeight=0.5)
    ActiveRegionSize=0.1
    Debug=(bDrawDebug=false,bDraw2D=False,bDrawLocalAxes=False,DebugScale=25,DebugSegments=20,DebugThickness=0.25)
    DriverAxis=(X=-1,Y=0,Z=0)
    DriverItem=(Type=Bone,Name="lowerarm_r")
    FalloffRegionScaleFactors=(PositiveWidth=1,NegativeWidth=1,PositiveHeight=1,NegativeHeight=1)
    FalloffSize=0.5
    FlipHeightScaling=False
    FlipWidthScaling=False
    OptionalParentItem=(Type=Bone,Name="None")
    RotationOffset=(X=-90,Y=180,Z=15)
   )
   > &RigUnit_ModifyTransforms(
    ItemToModify="((Item=(Type=Bone,Name=\"dyn_extra_upperarm_twist_r\"),Transform=(Rotation=(X=-0.2588,Y=0,Z=0,W=0.9659),Translation=(X=0,Y=0,Z=0),Scale3D=(X=1,Y=1,Z=1))),(Item=(Type=Bone,Name=\"dyn_extra_lowerarm_start_r\"),Transform=(Rotation=(X=0,Y=0,Z=-0.2588,W=0.9659),Tra ...(truncated)"
    WeightMinimum=0
    Weight=@n7.OutputParam
   )
  ]}
 New Function_ContainedGraph {nodes [
   n1 = RigVMFunctionEntryNode()
   RigVMFunction_Sequence = &RigVMFunction_Sequence(ExecuteContext=@n1.ExecuteContext)
   RigVMFunctionReturnNode(ExecuteContext=@RigVMFunction_Sequence.B)
   n2 = &RigVMFunction_MathQuaternionToEuler(
    RotationOrder=ZYX
    Value=&RigUnit_GetTransform(Item=(Type=Bone,Name="foot_l"), Space=LocalSpace, bInitial=False).Transform.Rotation
   )
   RigVMFunction_MathBoolToFloat = &RigVMFunction_MathBoolToFloat(Value=RigVMVariableNode() {variable=Skydive}.Value)
   AlphaInterp = &RigVMFunction_AlphaInterp(InterpSpeedDecreasing=10, InterpSpeedIncreasing=10, Scale=1, bClampResult=true, Value=@RigVMFunction_MathBoolToFloat.Result)
   &RigUnit_SetRotation(
    Item=(Type=Bone,Name="foot_l")
    bInitial=False
    bPropagateToChildren=true
    Value=&RigVMFunction_MathQuaternionFromEuler(
     Euler=(X=0,Y=0,Z=0)
     Euler.X=&RigVMFunction_MathDoubleMul(A=1, B=@n2.Result.X).Result
     Euler.Y=&RigVMFunction_MathDoubleMul(A=1, B=@n2.Result.Y).Result
     Euler.Z=&RigVMFunction_MathDoubleMul(A=0.75, B=@n2.Result.Z).Result
    ).Result
    Weight=@AlphaInterp.Result
    ExecutePin=@RigVMFunction_Sequence.A
   )
   n3 = &RigVMFunction_MathQuaternionToEuler(
    RotationOrder=ZYX
    Value=&RigUnit_GetTransform(Item=(Type=Bone,Name="foot_r"), Space=LocalSpace, bInitial=False).Transform.Rotation
   )
   > &RigUnit_SetRotation(
    Item=(Type=Bone,Name="foot_r")
    bInitial=False
    bPropagateToChildren=true
    Value=&RigVMFunction_MathQuaternionFromEuler(
     Euler=(X=0,Y=0,Z=0)
     Euler.Z=&RigVMFunction_MathDoubleMul(A=0.75, B=@n3.Result.Z).Result
     Euler.X=&RigVMFunction_MathDoubleMul(A=0.5, B=@n3.Result.X).Result
     Euler.Y=&RigVMFunction_MathDoubleMul(A=0.5, B=@n3.Result.Y).Result
    ).Result
    Weight=@AlphaInterp.Result
   )
   > &RigUnit_ModifyTransforms(
    ItemToModify="((Item=(Type=Bone,Name=\"foot_l\"),Transform=(Rotation=(X=-0.173,Y=-0.0858,Z=-0.0151,W=0.9811),Translation=(X=0,Y=0,Z=0),Scale3D=(X=1,Y=1,Z=1))),(Item=(Type=Bone,Name=\"foot_r\"),Transform=(Rotation=(X=-0.1006,Y=-0.0704,Z=-0.1798,W=0.976),Translation=(X=0,Y=-1 ...(truncated)"
    WeightMinimum=0
    Weight=@AlphaInterp.Result
   )
  ]}
 RigVMFunctionLibrary {nodes [
   Setup Arms = RigVMCollapseNode(ExecuteContext=())
   Setup Foot = RigVMCollapseNode(ExecuteContext=())
   Setup Leg = RigVMCollapseNode(ExecuteContext=())
  ]}
 RigVMFunction_Sequence_1_ContainedGraph {nodes [
   n1 = RigVMFunctionEntryNode()
   RigVMFunction_Sequence = &RigVMFunction_Sequence(ExecuteContext=@n1.ExecuteContext)
   n2 = &RigVMFunction_Sequence(ExecuteContext=@RigVMFunction_Sequence.B)
   n3 = &RigVMFunction_Sequence(ExecuteContext=@n2.B)
   n4 = &RigVMFunction_Sequence(ExecuteContext=@n3.B)
   RigVMFunctionReturnNode(A=@RigVMFunction_Sequence.A, B=@n2.A, C=@n3.A, F=@n4.A, G=@n4.B)
  ]}
 RigVMFunction_Sequence_2_ContainedGraph {nodes [
   n1 = RigVMFunctionEntryNode()
   RigVMFunction_Sequence = &RigVMFunction_Sequence(ExecuteContext=@n1.ExecuteContext)
   n2 = &RigVMFunction_Sequence(ExecuteContext=@RigVMFunction_Sequence.B)
   n3 = &RigVMFunction_Sequence(ExecuteContext=@n2.B)
   n4 = &RigVMFunction_Sequence(ExecuteContext=@n3.B)
   RigVMFunctionReturnNode(A=@RigVMFunction_Sequence.A, B=@n2.A, C=@n3.A, D=@n4.A, E=@n4.B)
  ]}
 RigVMModel {nodes [
   n1 = &RigUnit_PrepareForExecution()
   &RigUnit_HierarchyImportFromSkeleton(NameSpace=None, bIncludeCurves=False, bIncludeMeshSockets=False, bIncludeVirtualBones=true, ExecutePin=@n1.ExecutePin)
   RigUnit_BeginExecution = &RigUnit_BeginExecution()
   Setup Leg = RigVMFunctionReferenceNode(ExecuteContext=@RigUnit_BeginExecution.ExecutePin)
   > Setup Arms = RigVMFunctionReferenceNode()
   addFKCtrl = RigVMFunctionReferenceNode(
    Custom_FKCtrl=()
    ExecuteContext=()
    Options_FKCtrl="(pelvis_1_#1=true,spine_3_#2=true,clavicle_5_#3=true,arm_7_#4=true,leg_9_#5=true,head_11_46DC15EF465CD2979 ...(truncated)"
    Shape_Settings="(pelvis_2_#6=(bVisible=True,Name=\"Hexagon_Thin\",Color=(R=0.0117,G=0.8281,B=1,A=1),Transform=(Rotation=(X=0,Y=-0.7071,Z=0,W=0.7071),Translation=(X=2,Y=0,Z=0),Scale3D=(X=5,Y=5,Z=7))),spine_4_#7=(bVi ...(truncated)"
   )
   setFKCtrl = RigVMFunctionReferenceNode(ExecuteContext=(), setFK_Ctrl=false)
  ]}
]
variables [
 Crouching {type=double, default=0}
 CrouchingBool {type=bool, default=False}
 Emote {type=bool, default=False}
 Idle {type=double, default=0}
 Leg_Front_L {type=double, default=0}
 Leg_Front_R {type=double, default=0}
 Leg_Side_L {type=double, default=0}
 Leg_Side_R {type=double, default=0}
 Skydive {type=bool, default=False}
 Twist_In_L {type=double, default=0}
 Twist_In_R {type=double, default=0}
 Twist_Out_L {type=double, default=0}
 Twist_Out_R {type=double, default=0}
]
