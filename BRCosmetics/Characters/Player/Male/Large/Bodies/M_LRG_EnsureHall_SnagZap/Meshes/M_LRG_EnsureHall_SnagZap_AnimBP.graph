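# Animation Blueprint for the SnagZap large male player body cosmetic, following the parent mesh via Copy Pose From Mesh.
# A Control Rig takes emote, skydive, crouch and idle inputs, then fixed additive bone offsets adjust a pelvis flap, lower back, shoulder armor and backpack attachment.
# The update event derives idle from movement and flags the celebration montage and a name-matched Snag backpack.
# The celebration and backpack flags are not read in these graphs.
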AnimBlueprint /BRCosmetics/Characters/Player/Male/Large/Bodies/M_LRG_EnsureHall_SnagZap/Meshes/M_LRG_EnsureHall_SnagZap_AnimBP
target_skeleton=&/M_LRG_EnsureHall_SnagZap_Skeleton
graphs [
 AnimGraph {nodes [
   n1 = Control Rig!(
    Source=Copy Pose From Mesh() {bUseAttachedParent=true}
    Alpha
    Emote:bool=Get bIsPlayingEmote()
    Skydive:bool=Get bIsSkyDiving()
    CrouchingBool:bool=Get bIsCrouching()
    Idle:double=Get Alpha_Idle() -> (Alpha_Idle:double)
   ) -> (Pose) {
    alphaBoolBlend {blendOption=Linear}
    controlRigAssetReference {blueprintRigClass=&/M_LRG_EnsureHall_SnagZap_CtrlRig.M_LRG_EnsureHall_SnagZap_CtrlRig_C}
   }
   n2 = "Transform (Modify) Bone - Bone: PelvisFlaps_l"!(
    ComponentPose=Local To Component(LocalPose=@n1)
    bAlphaBoolEnabled:bool=True
    Translation="1.407,0.000,0.000"
    Rotation="16.350,6.116,-0.000"
    Scale="1,1,1"
   ) -> (Pose) {
    alphaInputType=Bool
    rotationMode=BMM_Additive
    rotationSpace=BCS_BoneSpace
    translationMode=BMM_Additive
    translationSpace=BCS_BoneSpace
    alphaBoolBlend {blendInTime=0.5, blendOption=Linear, blendOutTime=0.2}
    boneToModify {boneName=PelvisFlaps_l}
    rotation=<rot 16.35 6.116 -0.0001>
    translation {x=1.407}
   }
   n3 = "Transform (Modify) Bone - Bone: back_low"!(ComponentPose=@n2, bAlphaBoolEnabled:bool=True, Translation="2.884,-0.000,0.363", Rotation="-10.839,-0.000,0.000", Scale="0.691,1.000,1.000") -> (Pose) {
    alphaInputType=Bool
    rotationMode=BMM_Additive
    rotationSpace=BCS_BoneSpace
    scaleMode=BMM_Additive
    scaleSpace=BCS_BoneSpace
    translationMode=BMM_Additive
    translationSpace=BCS_BoneSpace
    alphaBoolBlend {blendInTime=0.5, blendOption=Linear, blendOutTime=0.2}
    boneToModify {boneName=back_low}
    rotation=<rot -10.8391 0 0>
    scale {x=0.691}
    translation=<2.8836 0 0.3631>
   }
   n4 = "Transform (Modify) Bone - Bone: dyn_shoulder_armor_l"(ComponentPose=@n3, Alpha=1, Translation="0.235,0,6.235", Rotation="0,0,0", Scale="1,1,1") {
    rotationMode=BMM_Additive
    rotationSpace=BCS_BoneSpace
    scaleMode=BMM_Additive
    scaleSpace=BCS_BoneSpace
    translationMode=BMM_Additive
    translationSpace=BCS_BoneSpace
    alphaBoolBlend {blendInTime=0.5, blendOption=Linear, blendOutTime=0.2}
    alphaScaleBiasClamp {bInterpResult=true}
    boneToModify {boneName=dyn_shoulder_armor_l}
    translation {x=0.235, z=6.235}
   }
   n5 = "Transform (Modify) Bone - Bone: dyn_shoulder_armor_r"(ComponentPose=@n4, Alpha=1, Translation="-0.235,-0.000,-6.235", Rotation="0,0,0", Scale="1,1,1") {
    rotationMode=BMM_Additive
    rotationSpace=BCS_BoneSpace
    scaleMode=BMM_Additive
    scaleSpace=BCS_BoneSpace
    translationMode=BMM_Additive
    translationSpace=BCS_BoneSpace
    alphaBoolBlend {blendInTime=0.5, blendOption=Linear, blendOutTime=0.2}
    alphaScaleBiasClamp {bInterpResult=true}
    boneToModify {boneName=dyn_shoulder_armor_r}
    translation=<-0.235 0 -6.2348>
   }
   n6 = "Transform (Modify) Bone - Bone: attach_backpack"(ComponentPose=@n5, Alpha=1, Translation="-0.001,1.400,0.646", Rotation="0,0,0", Scale="1,1,1") {
    rotationMode=BMM_Additive
    rotationSpace=BCS_BoneSpace
    scaleMode=BMM_Additive
    scaleSpace=BCS_BoneSpace
    translationMode=BMM_Additive
    translationSpace=BCS_BoneSpace
    alphaBoolBlend {blendInTime=0.5, blendOption=Linear, blendOutTime=0.2}
    alphaScaleBiasClamp {bInterpResult=true}
    boneToModify {boneName=attach_backpack}
    translation=<-0.001 1.4 0.646>
   }
   Output Pose(Result=Component To Local(ComponentPose=@n6)) {layerGroup=None}
  ]}
 EventGraph {nodes [
   Event BlueprintUpdateAnimation()
   > n1 = Sequence()
   n2 = Get bIsMoving2D()
   > Set Alpha_Idle(Alpha_Idle:double="To Float (Boolean)"(InBool=NOT Boolean(A=@n2))) -> (Output_Get:double)
   n3 = GetAnimInstance(self=GetOwnerSkeletalMeshComponent())
   > Set PlayingCelebration(PlayingCelebration:double="To Float (Boolean)"(InBool=Montage_IsPlaying(self=@n3, Montage=/Game/Animation/Game/MainPlayer/Emotes/Emote_Celebration_M))) -> (Output_Get:double)
   n4 = Cast To FortPlayerPawn(execute=@n1.then_1, Object=TryGetPawnOwner())
   n5 = GetDisplayName(Object=GetSkeletalMeshForPartType(self=@n4, PartType=Backpack))
   > Set Detected BP(Detected BP:double="To Float (Boolean)"(InBool=MatchesWildcard(SourceString=@n5, Wildcard=*M_LRG_EnsureHall_Snag_Pack*, SearchCase=CaseSensitive))) -> (Output_Get:double)
  ]}
]
variables [
 Alpha_Idle {type=number}
 Detected BP {type=number}
 PlayingCelebration {type=number}
]
