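# Animation Blueprint for the face accessory part of the SnagZap large male player cosmetic body.
# It copies the pose from the attached parent mesh, then runs the part's own Control Rig on it to produce the output pose.
# The rig receives emote, skydive-floating and skydiving state flags as inputs; the update event has no logic attached.
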AnimBlueprint /BRCosmetics/Characters/Player/Male/Large/Bodies/M_LRG_EnsureHall_SnagZap/Meshes/Parts/M_LRG_EnsureHall_SnagZap_FaceAcc_AnimBP
target_skeleton=&/M_LRG_EnsureHall_SnagZap_FaceAcc_Skeleton
graphs [
 AnimGraph {nodes [
   n1 = Control Rig!(
    Source=Copy Pose From Mesh() {bUseAttachedParent=true}
    Alpha
    EmoteBool:bool=Get bIsPlayingEmote()
    FloatingBool:bool=Get bIsSkydiveFloating()
    SkydiveBool:bool=Get bIsSkyDiving()
   ) -> (Pose) {
    alphaBoolBlend {blendOption=Linear}
    controlRigAssetReference {blueprintRigClass=&/M_LRG_EnsureHall_SnagZap_FaceAcc_CtrlRig.M_LRG_EnsureHall_SnagZap_FaceAcc_CtrlRig_C}
   }
   Output Pose(Result=@n1) {layerGroup=None}
  ]}
 EventGraph {nodes [Event BlueprintUpdateAnimation()]}
]
variables []
